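# MSP_Effect is an example MetaSound patch in the EffectsChain examples that applies distortion to stereo audio.
# It implements the stereo input and output format interfaces, so audio enters and leaves as left and right channels.
# Each channel runs through its own Harmonix Distortion node, left at default settings, straight to the matching output, with no parameters exposed beyond the audio pins.
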MetaSoundPatch /MetasoundPolyphonyInternal/Examples/EffectsChain/MSP_Effect
inputs [
 "UE.InputFormat.Stereo.Audio:0" {type=Audio}
 "UE.InputFormat.Stereo.Audio:1" {type=Audio}
]
interfaces [UE.InputFormat.Stereo:1.0, UE.OutputFormat.Stereo:1.0]
outputs [
 "UE.OutputFormat.Stereo.Audio:0" {type=Audio}
 "UE.OutputFormat.Stereo.Audio:1" {type=Audio}
]
pages [Default {
  input_defaults {"UE.InputFormat.Stereo.Audio:0"=null, "UE.InputFormat.Stereo.Audio:1"=null}
  nodes [
   HarmonixNodesDistortion_0 = HarmonixNodes.Distortion(In=^"UE.InputFormat.Stereo.Audio:0") {also {version [0,1]}}
   HarmonixNodesDistortion_1 = HarmonixNodes.Distortion(In=^"UE.InputFormat.Stereo.Audio:1") {also {version [0,1]}}
  ]
  links [
   {dst="@UE.OutputFormat.Stereo.Audio:0", src=HarmonixNodesDistortion_0.Out}
   {dst="@UE.OutputFormat.Stereo.Audio:1", src=HarmonixNodesDistortion_1.Out}
  ]
 }]
