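# MSP_HPF is a MetaSound patch in the MetasoundPolyphonyInternal EffectsChain examples that holds a high-pass filter.
# A single float input, Cutoff Frequency, defaulting to zero, drives the cutoff of a One-Pole High Pass Filter node.
# The patch declares no audio input or outputs, so the filter's audio pins are unconnected within this graph.
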MetaSoundPatch /MetasoundPolyphonyInternal/Examples/EffectsChain/MSP_HPF
inputs [Cutoff Frequency {type=Float}]
outputs []
pages [Default {
  input_defaults {Cutoff Frequency=0}
  nodes [OnePoleHighPassFilter_0 = UE.One-Pole High Pass Filter:Audio(Cutoff Frequency=^Cutoff Frequency)]
 }]
