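# MetaSound patch from the effects chain examples that applies a high-pass then low-pass filter to stereo audio.
# Each channel runs through its own one-pole high-pass filter into a one-pole low-pass filter, then out through the stereo output interface.
# Two float inputs expose the high-pass and low-pass cutoff frequencies, each shared by both channels.
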MetaSoundPatch /MetasoundPolyphonyInternal/Examples/EffectsChain/MSP_HPF_LPF_Stereo
inputs [
 HPF.Cutoff Frequency {type=Float}
 LPF.Cutoff Frequency {type=Float}
 "UE.InputFormat.Stereo.Audio:0" {type=Audio}
 "UE.InputFormat.Stereo.Audio:1" {type=Audio}
]
interfaces [UE.InputFormat.Stereo:1.0, UE.OutputFormat.Stereo:1.0]
outputs [
 "UE.OutputFormat.Stereo.Audio:0" {type=Audio}
 "UE.OutputFormat.Stereo.Audio:1" {type=Audio}
]
pages [Default {
  input_defaults {HPF.Cutoff Frequency=14642.6807, LPF.Cutoff Frequency=3679.9985, "UE.InputFormat.Stereo.Audio:0"=null, "UE.InputFormat.Stereo.Audio:1"=null}
  nodes [
   OnePoleHighPassFilter_1 = UE.One-Pole High Pass Filter:Audio(Cutoff Frequency=^HPF.Cutoff Frequency, In=^"UE.InputFormat.Stereo.Audio:1")
   OnePoleLowPassFilter_0 = UE.One-Pole Low Pass Filter:Audio(Cutoff Frequency=^LPF.Cutoff Frequency, In=@OnePoleHighPassFilter_1.Out)
   OnePoleHighPassFilter_0 = UE.One-Pole High Pass Filter:Audio(Cutoff Frequency=^HPF.Cutoff Frequency, In=^"UE.InputFormat.Stereo.Audio:0")
   OnePoleLowPassFilter_1 = UE.One-Pole Low Pass Filter:Audio(Cutoff Frequency=^LPF.Cutoff Frequency, In=@OnePoleHighPassFilter_0.Out)
  ]
  links [
   {dst="@UE.OutputFormat.Stereo.Audio:1", src=OnePoleLowPassFilter_0.Out}
   {dst="@UE.OutputFormat.Stereo.Audio:0", src=OnePoleLowPassFilter_1.Out}
  ]
 }]
