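# MSP_LPF is a MetaSound patch that applies a stereo low-pass filter, sitting among the MetasoundPolyphonyInternal plugin's effects-chain examples.
# Each input channel runs through its own one-pole low-pass filter to the matching stereo output, with both filters sharing one Cutoff Frequency input.
# It also exposes a MIDI Clock input, which is not connected to anything in the graph.
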MetaSoundPatch /MetasoundPolyphonyInternal/Examples/EffectsChain/MSP_LPF
inputs [
 Cutoff Frequency {type=Float}
 MIDI Clock {type=MIDIClock}
 "UE.InputFormat.Stereo.Audio:0" {type=Audio}
 "UE.InputFormat.Stereo.Audio:1" {type=Audio}
]
interfaces [UE.InputFormat.Stereo:1.0, UE.OutputFormat.Stereo:1.0]
outputs [
 "UE.OutputFormat.Stereo.Audio:0" {type=Audio}
 "UE.OutputFormat.Stereo.Audio:1" {type=Audio}
]
pages [Default {
  input_defaults {Cutoff Frequency=502.3954, MIDI Clock=null, "UE.InputFormat.Stereo.Audio:0"=null, "UE.InputFormat.Stereo.Audio:1"=null}
  nodes [
   OnePoleLowPassFilter_0 = UE.One-Pole Low Pass Filter:Audio(Cutoff Frequency=^Cutoff Frequency, In=^"UE.InputFormat.Stereo.Audio:0")
   OnePoleLowPassFilter_1 = UE.One-Pole Low Pass Filter:Audio(Cutoff Frequency=^Cutoff Frequency, In=^"UE.InputFormat.Stereo.Audio:1")
  ]
  links [
   {dst="@UE.OutputFormat.Stereo.Audio:0", src=OnePoleLowPassFilter_0.Out}
   {dst="@UE.OutputFormat.Stereo.Audio:1", src=OnePoleLowPassFilter_1.Out}
  ]
 }]
