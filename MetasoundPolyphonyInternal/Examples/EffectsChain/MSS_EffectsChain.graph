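# Example stereo MetaSound source demonstrating the EffectsChain node from MetasoundPolyphonyInternal on a blues guitar riff.
# On play, a wave player feeds both channels into the EffectsChain node, which is configured by an effects container asset and drives the stereo output.
# A second wave player with the same riff also triggers on play but is not connected to the output, and the Tempo input is unused.
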MetaSoundSource /MetasoundPolyphonyInternal/Examples/EffectsChain/MSS_EffectsChain
inputs [
 Tempo {type=Float}
 UE.Source.OnPlay {type=Trigger}
]
interfaces [UE.OutputFormat.Stereo:1.0, UE.Source:1.1]
outputs [
 "UE.OutputFormat.Stereo.Audio:0" {type=Audio}
 "UE.OutputFormat.Stereo.Audio:1" {type=Audio}
]
pages [Default {
  input_defaults {Tempo=120, UE.Source.OnPlay=null}
  nodes [
   WavePlayer_0 = UE.Wave Player:Stereo(Play=^UE.Source.OnPlay) {also {
     defaults {Loop="<MetaSoundBuilderResult.SUCCEEDED: 0>", Wave Asset=&/g_blues_riff_120bpm-78602}
     version [1,1]
    }}
   MetasoundPolyphonyInternalEffectsChain_0 = MetasoundPolyphonyInternal.EffectsChain(In Left=@WavePlayer_0.Out Left, In Right=@WavePlayer_0.Out Right) {also {defaults {Effects Container=&/NewMetaSoundEffectsContainer}, version [0,1]}}
   WavePlayer_1 = UE.Wave Player:Stereo(Play=^UE.Source.OnPlay) {also {
     defaults {Loop="<MetaSoundBuilderResult.SUCCEEDED: 0>", Wave Asset=&g_blues_riff_120bpm-78602}
     version [1,1]
    }}
  ]
  links [
   {dst="@UE.OutputFormat.Stereo.Audio:0", src=MetasoundPolyphonyInternalEffectsChain_0.Out Left}
   {dst="@UE.OutputFormat.Stereo.Audio:1", src=MetasoundPolyphonyInternalEffectsChain_0.Out Right}
  ]
 }]
properties {
 assetImportData=&self:AssetImportData
 bLooping=true
 duration=10000
 numChannels=2
 outputFormat=Stereo
 sampleRate=48000
}
